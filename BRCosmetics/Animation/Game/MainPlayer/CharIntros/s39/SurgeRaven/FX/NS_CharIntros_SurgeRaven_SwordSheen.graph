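# Niagara system for a sword sheen in the Surge Raven main-player character intro, built from one burst-spawned mesh emitter, SharpShine.
# A curve drives a dynamic material parameter upward over each short-lived, bright particle's life, alongside fixed companion parameters, to animate the shine in the material.
# Alpha scales by time of day, dimmest at night.
# Front-end, preview and swinging user booleans are declared, with no visible use in the listed modules.
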NiagaraSystem /BRCosmetics/Animation/Game/MainPlayer/CharIntros/s39/SurgeRaven/FX/NS_CharIntros_SurgeRaven_SwordSheen
dependencies {usedDataInterfaces [&NiagaraDataInterfaceCurve], usedDynamicInputs [], usedModules [&SystemState]}
emitters {SharpShine {
  emitterUpdateScript [
   &EmitterState()
   &SpawnBurst_Instantaneous(Use Spawn Probability=true, UseLoopCountLimit=false)
  ]
  particleSpawnScript [&InitializeParticle(Color=<rgba 100 100 100 0.8>, Lifetime=1.5, UsePositionOffset=false, Write Material Random=true)]
  particleUpdateScript [
   &ParticleState()
   &SolveForcesAndVelocity(Clamp Velocity=false, Limit Acceleration=false)
   &DynamicMaterialParameters(
    Index 0 Param 1=&FloatFromCurve(FloatCurve=DataInterface() {propertyValues {Curve {keys [
         {arriveTangent=1.7711, interpMode=RCIM_Cubic, leaveTangent=1.7711, leaveTangentWeight=0.3333, tangentMode=RCTM_User, time=0, value=0.2}
         {arriveTangent=0.0928, arriveTangentWeight=0.3333, interpMode=RCIM_Cubic, leaveTangent=0.0928, tangentMode=RCTM_User, time=1, value=1}
        ]}}})
    Index 0 Param 2=0.15
    Index 0 Param 3=-0.18
    Module.Index0Param0Write=true
    Module.Index0Param1Write=true
    Module.Index0Param2Write=true
    Module.Index0Param3Write=true
   )
   &ScaleColor(
    Scale Alpha=&FNTimeOfDayScalar(Night Scale=0.3, Morning Scale=0.7, Evening Scale=0.7)
    ScaleA=true
    ScaleRGB=false
   )
  ]
  rendererClasses [&NiagaraMeshRendererProperties]
  renderers [{rendererClass=&NiagaraMeshRendererProperties, rendererIndex=0}]
 }}
summary {
 systemName=NS_CharIntros_SurgeRaven_SwordSheen
 emitters [{emitterName=SharpShine, rendererClasses [&NiagaraMeshRendererProperties]}]
 userVariables [
  User.bIsFrontEnd {type=NiagaraBool}
  User.bIsFrontEndPreview {type=NiagaraBool}
  User.bisSwinging {type=NiagaraBool}
 ]
}
